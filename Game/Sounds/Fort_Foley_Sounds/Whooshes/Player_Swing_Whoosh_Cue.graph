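# Sound cue for the player's swing whoosh foley, routed to the melee weapon sound class.
# A Random node picks one of five equally weighted stereo whoosh waves, and a Modulator varies pitch while holding volume constant.
# The cue raises overall pitch, lowers volume, and overrides attenuation and concurrency, blocking new instances when the concurrency limit is reached.
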SoundCue /Game/Sounds/Fort_Foley_Sounds/Whooshes/Player_Swing_Whoosh_Cue
root=Modulator_2
nodes [
 {id=Modulator_2, children [Random_0], properties {pitchMax=1.2, volumeMax=1, volumeMin=1}}
 {id=Random_0, children [WavePlayer_0,WavePlayer_1,WavePlayer_2,WavePlayer_3,WavePlayer_4], properties {weights [1,1,1,1,1]}}
 {id=WavePlayer_0, properties {soundWaveAssetPtr=&/player_swing_whoosh_stereo_01}}
 {id=WavePlayer_1, properties {soundWaveAssetPtr=&/player_swing_whoosh_stereo_02}}
 {id=WavePlayer_2, properties {soundWaveAssetPtr=&/player_swing_whoosh_stereo_03}}
 {id=WavePlayer_3, properties {soundWaveAssetPtr=&/player_swing_whoosh_stereo_04}}
 {id=WavePlayer_4, properties {soundWaveAssetPtr=&/player_swing_whoosh_stereo_05}}
]
properties {
 bOverrideAttenuation=true
 bOverrideConcurrency=true
 maxDistance=4000
 pitchMultiplier=1.2
 soundClassObject=&Fort_WeaponMelee
 volumeMultiplier=0.6
 concurrencyOverrides {resolutionRule=PreventNew}
}
